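# Translucent, lit GPU sprite particle material for snow chunks, built on smoke textures with a matching normal map.
# Each particle picks a random 2x2 sub-UV cell; the red channel, particle color and softened time-of-day tint form a heavily darkened color.
# Particle alpha, depth fade and near-camera fade drive opacity, which morphs from texture alpha to a soft radial mask as it fades.
# Mobile feature levels skip both fades.
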Material /Figment_S07/Effects/Fort_Effects/Materials/Water/M_Snow_Chunk_GPU
graph {
 properties {
  bUsedWithParticleSprites=true
  blendMode=BLEND_Translucent
  pixelDepthOffsetMode=PDOM_Legacy
  translucencyLightingMode=TLM_VolumetricPerVertexDirectional
  usedShadingModels=Default Lit
 }
 nodes [
  n1 = LinearInterpolate(B=MaterialFunctionCall() {materialFunction=&Color_Time_of_Day}.Result) {constA=1}
  n2 = ParticleColor()
  n3 = Constant() {r=2}
  n4 = MaterialFunctionCall!(UseTexCoord 1, Rows=@n3, Columns=@n3, Particle Random Value=ParticleRandom()) {materialFunction=/Game/Effects/Fort_Effects/Materials/_Functions/MF_RandomSubUV_GPU_Sprites}
  n5 = TextureSample(UVs=@n4.Result) {samplerType=SAMPLERTYPE_LinearColor, texture=/Game/Effects/Fort_Effects/Textures/Smoke/SmokeTest2_Soft}
  Multiply(A=Multiply(A=Multiply(A=@n1, B=@n2.RGB)) {constB=0.05}, B=@n5.R)
  n6 = DepthFade(
   Opacity=MaterialFunctionCall(Distance=ScalarParameter() {defaultValue=150, parameterName=Near Camera Fade Distance}) {materialFunction=&Near_Camera_Fade}.Result
   FadeDistance=ScalarParameter() {defaultValue=55, parameterName=DepthFadeDist}
  )
  n7 = Constant() {r=1}
  n8 = Multiply(A=@n2.A, B=FeatureLevelSwitch(Default=@n6, ES2=@n7, ES3_1=@n7, SM4_REMOVED=@n6))
  n9 = TextureCoordinate()
  n10 = Constant() {r=0.5}
  Multiply(A=@n8, B=LinearInterpolate(A=SphereMask(A=@n9, B=@n10, Radius=@n10) {hardnessPercent=0}, B=@n5.A, Alpha=Power(Base=@n8) {constExponent=4}))
  TextureSample(UVs=@n4.Result) {samplerType=SAMPLERTYPE_Normal, texture=/Game/Effects/Fort_Effects/Textures/Smoke/SmokeTest_N}
 ]
}
